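# Physics asset for the Glider_028_Reaper mesh, a Reaper glider cosmetic in the VinderTech GliderChute gadget set.
# It contains a single body on the Frame bone, built from one large, rotated box that serves as a coarse collision shape.
# There are no constraints, so the asset defines no jointed or ragdoll motion between bones.
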PhysicsAsset /BRCosmetics/Gadgets/Assets/VinderTech_GliderChute/Glider_Reaper/Meshes/Glider_028_Reaper_Physics
bodies {Frame [{shapeType=Box, extentX=154.6474, extentY=95.7386, extentZ=436.7194, center=<0.1946 -12.3016 -10.5937>, rotation=<rot 0 89.7313 -89.4851>}]}
constraints []
